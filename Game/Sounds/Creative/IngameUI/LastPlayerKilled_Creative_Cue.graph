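# Sound cue for the last-player-killed moment in Creative, judging by its name, in the in-game UI sound set.
# A mixer layers the last-player-killed SFX at full level with a quieter hornless Victory Royale surround music stinger.
# It plays boosted through the Victory Royale UI sound class, with a very large max distance.
# A concurrency override allows one instance at a time and blocks new ones while it plays.
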SoundCue /Game/Sounds/Creative/IngameUI/LastPlayerKilled_Creative_Cue
root=Mixer_0
nodes [
 {id=Mixer_0, children [WavePlayer_2,WavePlayer_4], properties {inputVolume [1,0.75]}}
 {id=WavePlayer_2, properties {soundWaveAssetPtr=/Game/Athena/Sounds/Character/SFX_LastPlayerKilled}}
 {id=WavePlayer_4, properties {soundWaveAssetPtr=/Game/Sounds/Fort_Music/Stingers/Victory_Royale_01_Surround_No_Horn}}
]
properties {
 bOverrideConcurrency=true
 maxDistance=2097152
 soundClassObject=/Game/Sounds/FortSoundClassesAndModes/Fort_UI_VictoryRoyale_PSM
 volumeMultiplier=3
 concurrencyOverrides {maxCount=1, resolutionRule=PreventNew}
}
