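# Creative-mode UI sound cue for a defeat when the last player is killed, routed through the Victory Royale UI sound class.
# A quality-level node sends its first two levels to a mixer layering the last-player-killed wave with a quieter victory-score defeat sting.
# The other two levels play a single optimized "newmix" defeat wave instead.
# A concurrency override allows one instance and blocks new plays while it runs.
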SoundCue /Game/Sounds/Creative/IngameUI/LastPlayerKilled_Creative_Defeated_Cue
root=QualityLevel_1
nodes [
 {id=Mixer_3, children [WavePlayer_3,WavePlayer_0], properties {inputVolume [1,0.4]}}
 {id=QualityLevel_1, children [Mixer_3,Mixer_3,WavePlayer_1,WavePlayer_1]}
 {id=WavePlayer_0, properties {soundWaveAssetPtr=&/UI_VictoryScore_Defeat_01}}
 {id=WavePlayer_1, properties {soundWaveAssetPtr=/Game/Athena/Sounds/Character/SFX_LastPlayerKilled_Defeated_opt_newmix}}
 {id=WavePlayer_3, properties {soundWaveAssetPtr=/Game/Athena/Sounds/Character/SFX_LastPlayerKilled}}
]
properties {
 bOverrideConcurrency=true
 maxDistance=2097152
 soundClassObject=/Game/Sounds/FortSoundClassesAndModes/Fort_UI_VictoryRoyale_PSM
 volumeMultiplier=1.26
 concurrencyOverrides {maxCount=1, resolutionRule=PreventNew}
}
