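# Sound cue for the in-game UI sound played when a player checkpoint triggers in Creative.
# A single wave player plays one checkpoint trigger wave at reduced volume in the Creative gadgets sound class.
# It uses the shared gadget-placement attenuation and overrides concurrency to cap simultaneous instances, stopping the quietest when exceeded.
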SoundCue /Game/Sounds/Creative/IngameUI/Player_Checkpoint_Trigger_Cue
root=WavePlayer_0
nodes [{id=WavePlayer_0, properties {soundWaveAssetPtr=&/Player_Checkpoint_Trigger_01}}]
properties {
 attenuationSettings=/Game/Sounds/Attenuation/Fort_GadgetPlace_Att
 bOverrideConcurrency=true
 maxDistance=2000
 soundClassObject=&Fort_Gadgets_Creative
 volumeMultiplier=0.5
 concurrencyOverrides {maxCount=2, resolutionRule=StopQuietest}
}
